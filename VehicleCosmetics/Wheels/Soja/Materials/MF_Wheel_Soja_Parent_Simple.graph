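# Simplified material function for Soja vehicle wheels that wraps MF_WheelParent_Simple and overrides world position offset, emissive and opacity.
# A time sine, masked by vertex color blue, oscillates vertices along local Y.
# Parent emissive is multiplied by an RGBA mask effect: R and G gate a shockwave curve texture panned with opposite V tiling, and B pulses with sine.
# The mask's alpha drives opacity.
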MaterialFunction /VehicleCosmetics/Wheels/Soja/Materials/MF_Wheel_Soja_Parent_Simple
graph {
 properties {description=Simplified MF For Wheels}
 nodes [
  Composite() {subgraphName=WPO Moving by Y axis}
  n1 = MaterialFunctionCall() {materialFunction=&MF_WheelParent_Simple}
  n2 = GetMaterialAttributes(None=@n1.MaterialAttributes) {attributeGetTypes [#1]}
  Composite!(Emissive=@n2.EmissiveColor) {subgraphName=Emissive FX}
  n3 = Sine(None=Time())
  n4 = Constant3Vector() {constant=<rgba 0 1 0 0>}
  n5 = Multiply(A=VertexColor().B, B=Multiply(A=ConstantBiasScale(None=@n3) {scale=1}, B=Transform(None=@n4) {transformSourceType=TRANSFORMSOURCE_Local}))
  MaterialExpressionReroute_23 = Reroute(None=@n5)
  MaterialExpressionReroute_27 = Reroute(None=@n2.EmissiveColor)
  n6 = TextureSampleParameter2D() {
   group=5_Special Effects
   parameterName=FX Emissive Masks
   samplerType=SAMPLERTYPE_LinearColor
   sortPriority=50
   texture=/VehicleCosmetics/Wheels/Soja/Textures/T_Soja_Rim_EmissiveMask
  }
  n7 = MaterialFunctionCall!(RGBA=@n6.RGBA) {materialFunction=/VehicleCosmetics/SharedMaterials/MaterialFunctions/MF_RGBA_Splitter}
  n8 = TextureObjectParameter() {
   group=5_Special Effects
   parameterName=FX
   samplerType=SAMPLERTYPE_LinearColor
   sortPriority=51
   texture=/VehicleCosmetics/Wheels/Soja/Textures/Curve_Shockwaves_Wrap
  }
  n9 = TextureSample(UVs=Panner(Coordinate=TextureCoordinate() {vTiling=-1}) {speedY=-1.5}, Tex=@n8) {
   samplerSource=SSM_Wrap_WorldGroupSettings
   samplerType=SAMPLERTYPE_LinearColor
   texture=&Curve_Shockwaves_Wrap
  }
  n10 = Multiply(A=@n7.R, B=Add(A=@n9.R, B=@n9.R))
  n11 = TextureSample(UVs=Panner(Coordinate=TextureCoordinate()) {speedY=-1.5}, Tex=@n8) {
   samplerSource=SSM_Wrap_WorldGroupSettings
   samplerType=SAMPLERTYPE_LinearColor
   texture=&Curve_Shockwaves_Wrap
  }
  n12 = Multiply(A=@n7.G, B=Add(A=@n11.R, B=@n11.R))
  n13 = Sine(None=Time())
  n14 = Multiply(A=@MaterialExpressionReroute_27.Emissive, B=Add(A=Add(A=@n10, B=@n12), B=Multiply(A=@n7.B, B=@n13)))
  MaterialExpressionReroute_28 = Reroute(None=@n14)
  MaterialExpressionReroute_29 = Reroute(None=@n6.A)
  FunctionOutput(None=SetMaterialAttributes!(
    MaterialAttributes=@n1.MaterialAttributes
    WorldPositionOffset=@MaterialExpressionReroute_23.WPO
    EmissiveColor=@MaterialExpressionReroute_28.Emissive
    Opacity=@MaterialExpressionReroute_29.Opacity
   ) {attributeSetTypes [#2, #1, #3]}) {outputName=MaterialAttributes}
  PinBase()
  PinBase!(WPO=@n5) {reroutePins [WPO {expression=&self:MaterialExpressionReroute_23}]}
  PinBase!() {reroutePins [Emissive {expression=&self:MaterialExpressionReroute_27}]}
  PinBase!(Emissive=@n14, Opacity=@n6.A) {reroutePins [
    Emissive {expression=&self:MaterialExpressionReroute_28}
    Opacity {expression=&self:MaterialExpressionReroute_29}
   ]}
 ]
}
